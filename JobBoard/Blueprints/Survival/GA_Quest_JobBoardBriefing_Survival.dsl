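; Gameplay ability for the Survival job board briefing quest that marks the quest instigator with tracking circles and reports failure on elimination.
; Once playlist data is ready, the server flags the instigator by tag, splits player controllers by team, and registers shrinking radius trackers with separate enemy and squad colors.
; On end, it removes tracking and shows the owning client an urgent failure message if the instigator or a non-instigator's squadmate died.
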(event Ability|EventActivateAbility
  ;[0 0 0] Activate ability - when granted, make a tracking cirlce for everyone on the map for the instigator
  (|CallParentFunction)
  (Class|FortAsyncActionCurrentPlaylistReady|PlaylistDataReadyAsync
    (:OnReady
      (Ability|HasAuthority
        (:True
          (bind _output_get (Variables|Default|SetbOwnerIsInstigator (GameplayTags|HasTag (GameplayTags|GetOwnedGameplayTags (CallFunction|GetAvatarasFortPlayerPawn) :!class /Script/GameplayTags.GameplayTagAssetInterface) (Variables|Default|GetInstigatorTag) true)))
          (if _output_get
            _)
          (CallFunction|CollectAllPlayerControllers)
          (CallFunction|InitTrackingInfo))))))

(event Custom|CallFailedMessage
  ;[0 0 0] On end ability - remove the tracking circle and handle a failed message if failed
  ;;[0.1823 0.1823 0.1823] Make sure ability has an owner before attempting to call the replicated event!
  (Utilities|IsValid (Ability|GetAvatarActorfromActorInfo)
    (:Is%20Valid
      (CallFunction|DisplayUrgentQuestFailedMessage))))

(event Ability|EventOnEndAbility
  (params (bWasCancelled :type Boolean))
  (bind _returnvalue (CallFunction|GetAvatarasFortPlayerPawn))
  ;[0 0 0] On end ability - remove the tracking circle and handle a failed message if failed
  (|CallParentFunction bWasCancelled)
  ;;[0.1458 0.1458 0.1458] Remove tracking circle
  (Ability|HasAuthority
    (:True
      (RadiusTracker|RemoveTrackingforActor (RadiusTracker|GetGameStateRadiusTrackerComponent) (Ability|GetAvatarActorfromActorInfo))
      (if (Variables|Default|GetbOwnerIsInstigator)
        (if (FortPawn|IsDead _returnvalue)
          ;;[0.1823 0.1823 0.1823] Make sure ability has an owner before attempting to call the replicated event!
          (Utilities|IsValid (Ability|GetAvatarActorfromActorInfo)
            (:Is%20Valid
              (CallFunction|DisplayUrgentQuestFailedMessage))))
        (else
          ;;[0.1771 0.1771 0.1771] If the ability owner is not the instigator, see if one of their squadmates died to end this ability and display a failed message.
          (bind _returnvalue_2 (Fortnite|Utilities|AttemptGetPlayerStateforPawn _returnvalue))
          (bind _asfort_player_state_athena (Utilities|Casting|CastToFortPlayerStateAthena _returnvalue_2))
          (bind _returnvalue_3 (Class|FortGameStateAthena|GetSquadMemberPlayerStates (Fortnite|Athena|GameState|GetGameStateAthenafromContext _returnvalue) _asfort_player_state_athena false))
          (bind _array_element (Utilities|Array|ForEachLoopwithBreak _returnvalue_3)
            (:LoopBody
              (if (Class|FortPlayerStateZone|IsPlayerDead _array_element)
                (CallFunction|CallFailedMessage)))))))))

(event Custom|DisplayUrgentQuestFailedMessage
  ;[0 0 0] Replicated to owning client - displays a failed message for the quest when it's poached by an enemy player
  (if (Variables|Default|GetbOwnerIsInstigator)
    (Variables|Default|SetTextSubtitle "NSLOCTEXT(\"[DEE036E6C1EB0D00FE54090A61640185]\", \"27862F6C4CA946E41401E4A0575380F8\", \"You were eliminated!\")")
    (else
      (Variables|Default|SetTextSubtitle "NSLOCTEXT(\"[DEE036E6C1EB0D00FE54090A61640185]\", \"D069BDBE4446DE303CF1D1AD691128EA\", \"Squadmate was eliminated!\")")))
  (Class|AthenaHUDContext|ShowUrgentQuestMessage (BlueprintContext|GetAthenaHUDContext) (Utilities|Struct|MakeUrgentQuestData (Variables|Default|GetSpecialEventTag) "(GameplayTags=)" 6e+01 true false false false false true "" (Variables|Default|GetTextQuestFailed) (Variables|Default|GetTextSubtitle))))

(fn CollectAllPlayerControllers
  ;[0 0 0] Collect all player controllers and track who is a squadmate and who is an enemy
  (Variables|Default|SetInstigatorTeam (Team|GetActorTeam (Ability|GetAvatarActorfromActorInfo)))
  (for _index (range (Game|GetNumPlayerStates))
    (Utilities|IsValid (Game|GetPlayerState _index)
      (:Is%20Valid
        (bind _asfort_player_controller (Utilities|Casting|CastToFortPlayerController (PlayerState|GetPlayerController (Game|GetPlayerState _index))))
        (if (== (Team|GetActorTeam (Game|GetPlayerState _index)) (Variables|Default|GetInstigatorTeam))
          (Utilities|Array|Add (Variables|Default|GetFriendlyPlayerControllers) _asfort_player_controller)
          (else
            (Utilities|Array|Add (Variables|Default|GetEnemyPlayerControllers) _asfort_player_controller)))))))

(fn CreateTrackingInfo
  (params (CircleColor :type Linear%20Color%20Structure) (TargetTrackers :type Array%20of%20Fort%20Player%20Controller%20Object%20References) (GroupingType :type ERadiusTrackingGroupingType%20Enum :default "Global"))
  ;[0 0 0] Create a tracking circle around the quest instigator for the passed in targets
  (bind _returnvalue (RadiusTracker|MakeRadiusTrackingInitializer (Math|Random|RandomFloatinRange 0.0 (Variables|Default|GetMinimum%20Radius)) (Variables|Default|GetMinimum%20Radius) (Variables|Default|GetMaximum%20Radius) 1e+01 (Variables|Default|GetRadius%20Amount%20to%20Shrink%20Each%20Update) CircleColor 0 (Ability|GetAvatarActorfromActorInfo) TargetTrackers GroupingType))
  (bind _returnvalue_2 (Fortnite|AddActorTrackingInfo (RadiusTracker|GetGameStateRadiusTrackerComponent) _returnvalue))
  (return _returnvalue_2))

(fn InitTrackingInfo
  ;[0 0 0] Pass a circle color and target trackers for enemies and squadmates to create the tracking circles
  (bind _returnvalue (CallFunction|CreateTrackingInfo (Variables|Default|GetEnemyCircleColor) (Variables|Default|GetEnemyPlayerControllers)))
  (Variables|Default|SetEnemyTrackingKey _returnvalue)
  (bind _returnvalue_2 (CallFunction|CreateTrackingInfo (Variables|Default|GetFriendlyCircleColor) (Variables|Default|GetFriendlyPlayerControllers) "Team"))
  (Variables|Default|SetFriendlyTrackingKey _returnvalue_2)
  (return))

(fn OnRep_bOwnerIsInstigator)
