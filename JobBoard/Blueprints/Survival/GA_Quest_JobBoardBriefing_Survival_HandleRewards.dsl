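; Gameplay ability for the Job Board survival briefing quest that spawns loot rewards when the tracked players survive until it ends.
; On activation, the server stores the pawn carrying InstigatorTag as PlayerTarget, checking the avatar first, then its squad members.
; On end, if neither target nor avatar is dead and the phase is not EndGame, SpawnRewards drops RewardTierGroup loot in front of the avatar.
; The end check ignores whether the ability was cancelled.
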(event Ability|EventActivateAbility
  (bind _returnvalue (CallFunction|GetAvatarasFortPlayerPawn))
  (|CallParentFunction)
  (Ability|HasAuthority
    (:True
      (if (GameplayTags|HasTag (GameplayTags|GetOwnedGameplayTags (CallFunction|GetAvatarasFortPlayerPawn) :!class /Script/GameplayTags.GameplayTagAssetInterface) (Variables|Default|GetInstigatorTag) true)
        (Variables|Default|SetPlayerTarget (CallFunction|GetAvatarasFortPlayerPawn))
        (else
          (bind _returnvalue_2 (Fortnite|Utilities|AttemptGetPlayerStateforPawn _returnvalue))
          (bind _asfort_player_state_athena (Utilities|Casting|CastToFortPlayerStateAthena _returnvalue_2))
          (bind _returnvalue_3 (Class|FortGameStateAthena|GetSquadMemberPlayerStates (Fortnite|Athena|GameState|GetGameStateAthenafromContext _returnvalue) _asfort_player_state_athena false))
          (bind _array_element (Utilities|Array|ForEachLoopwithBreak _returnvalue_3)
            (:LoopBody
              (if (GameplayTags|HasTag (GameplayTags|GetOwnedGameplayTags (PlayerState|GetCurrentPawn _array_element) :!class /Script/GameplayTags.GameplayTagAssetInterface) (Variables|Default|GetInstigatorTag) true)
                (Variables|Default|SetPlayerTarget (PlayerState|GetCurrentPawn _array_element))))))))))

(event Ability|EventOnEndAbility
  (params (bWasCancelled :type Boolean))
  (bind _returnvalue (BattleRoyaleGamePhaseLogic|GetBattleRoyaleGamePhaseLogicComponent))
  (|CallParentFunction bWasCancelled)
  (Ability|HasAuthority
    (:True
      (Utilities|IsValid _returnvalue
        (:Is%20Valid
          (if (Math|Boolean|ORBoolean (FortPawn|IsDead (Variables|Default|GetPlayerTarget)) (FortPawn|IsDead (CallFunction|GetAvatarasFortPlayerPawn)) (Utilities|Enum|Equal(Enum) (BattleRoyaleGamePhaseLogic|GetGamePhase _returnvalue) "EndGame"))
            (Development|PrintString "Player died; no rewards spawned" true true "(R=1.000000,G=0.000000,B=0.368968,A=1.000000)")
            (else
              (CallFunction|SpawnRewards))))))))

(event Custom|SpawnRewards
  (bind _returnvalue (CallFunction|GetAvatarasFortPlayerPawn))
  (Ability|HasAuthority
    (:True
      (Development|PrintString "Players survived; spawning rewards" true true "(R=1.000000,G=0.000000,B=0.368968,A=1.000000)")
      (Utilities|IsValid _returnvalue
        (:Is%20Valid
          (Class|AthenaFunctionLibrary|SpawnLootfromTable (Variables|Default|GetRewardTierGroup) 1 -1 (+ (Transformation|GetActorLocation _returnvalue) (* (Variables|Default|GetRewardSpawnOffsetFromPlayer) (Transformation|GetActorForwardVector _returnvalue))) (Transformation|GetActorForwardVector _returnvalue) 0 true true))))))
